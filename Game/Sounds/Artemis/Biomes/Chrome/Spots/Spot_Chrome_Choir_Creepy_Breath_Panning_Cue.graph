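# Sound cue for an outdoor ambient spot sound in the Chrome biome, playing creepy choir breath panning one-shots.
# A Random node picks one of six equally weighted wave variations, and all four Quality Level inputs share that node.
# A root Modulator varies the volume on each play.
# It uses the wind one-shot tile ambient attenuation and the Athena outdoor ambient sound class, at roughly half volume with bus sends disabled.
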SoundCue /Game/Sounds/Artemis/Biomes/Chrome/Spots/Spot_Chrome_Choir_Creepy_Breath_Panning_Cue
root=Modulator_0
nodes [
 {id=Modulator_0, children [QualityLevel_0], properties {volumeMin=0.8}}
 {id=QualityLevel_0, children [Random_0,Random_0,Random_0,Random_0]}
 {id=Random_0, children [WavePlayer_0,WavePlayer_1,WavePlayer_2,WavePlayer_3,WavePlayer_4,WavePlayer_5], properties {weights [1,1,1,1,1,1]}}
 {id=WavePlayer_0, properties {soundWaveAssetPtr=&/Spot_Chrome_Choir_Creepy_Breath_Panning_01}}
 {id=WavePlayer_1, properties {soundWaveAssetPtr=&/Spot_Chrome_Choir_Creepy_Breath_Panning_02}}
 {id=WavePlayer_2, properties {soundWaveAssetPtr=&/Spot_Chrome_Choir_Creepy_Breath_Panning_03}}
 {id=WavePlayer_3, properties {soundWaveAssetPtr=&/Spot_Chrome_Choir_Creepy_Breath_Panning_04}}
 {id=WavePlayer_4, properties {soundWaveAssetPtr=&/Spot_Chrome_Choir_Creepy_Breath_Panning_05}}
 {id=WavePlayer_5, properties {soundWaveAssetPtr=&/Spot_Chrome_Choir_Creepy_Breath_Panning_06}}
]
properties {
 attenuationSettings=/Game/Sounds/Attenuation/Fort_TileAmbient_OneShot_Wind_Att
 bEnableBusSends=false
 maxDistance=8000
 soundClassObject=&Fort_Ambient_Outdoors_Athena
 volumeMultiplier=0.501
 preEffectBusSends [{}]
}
