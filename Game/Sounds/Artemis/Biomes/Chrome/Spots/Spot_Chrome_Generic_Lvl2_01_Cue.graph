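# Sound cue for generic level 2 wind one-shot ambient spots in the Chrome biome.
# One random node picks among six equally weighted wind waves from two variation sets and feeds every slot of a quality-level node, so all quality levels play the same content.
# A modulator at the root varies the volume.
# The cue plays through wind one-shot attenuation in the ambient one-shot class, slightly pitched down, with a distant reverb submix send configured.
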SoundCue /Game/Sounds/Artemis/Biomes/Chrome/Spots/Spot_Chrome_Generic_Lvl2_01_Cue
root=Modulator_0
nodes [
 {id=Modulator_0, children [QualityLevel_0], properties {volumeMin=0.85}}
 {id=QualityLevel_0, children [Random_2,Random_2,Random_2,Random_2]}
 {
  id=Random_2
  children [WavePlayer_18, WavePlayer_19, WavePlayer_20, WavePlayer_21, WavePlayer_22, WavePlayer_23]
  properties {weights [1,1,1,1,1,1]}
 }
 {id=WavePlayer_18, properties {soundWaveAssetPtr=&/Spot_Wind_Chrome_Generic_Lvl2_01_01}}
 {id=WavePlayer_19, properties {soundWaveAssetPtr=&/Spot_Wind_Chrome_Generic_Lvl2_01_02}}
 {id=WavePlayer_20, properties {soundWaveAssetPtr=&/Spot_Wind_Chrome_Generic_Lvl2_01_03}}
 {id=WavePlayer_21, properties {soundWaveAssetPtr=&/Spot_Wind_Chrome_Generic_Lvl2_02_01}}
 {id=WavePlayer_22, properties {soundWaveAssetPtr=&/Spot_Wind_Chrome_Generic_Lvl2_02_02}}
 {id=WavePlayer_23, properties {soundWaveAssetPtr=&/Spot_Wind_Chrome_Generic_Lvl2_02_03}}
]
properties {
 attenuationSettings=/Game/Sounds/Attenuation/Fort_TileAmbient_OneShot_Wind_Att
 bEnableBusSends=false
 maxDistance=8000
 pitchMultiplier=0.9
 soundClassObject=&Fort_Ambient_OneShots
 volumeMultiplier=0.401
 preEffectBusSends [{}]
 soundSubmixSends [{minSendLevel=0.2, sendLevel=0, soundSubmix=&Fortnite_Reverb_Distant}]
}
