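# Sound cue for a granulated horror whoosh ambient one-shot spot in the Chrome biome.
# A Random node picks among three equally weighted wave variations and is excluded from branch culling.
# All four quality-level branches share that same Random node, so quality does not change the content.
# It plays at reduced volume with wind one-shot attenuation, the ambient one-shots sound class and a distant reverb submix send.
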SoundCue /Game/Sounds/Artemis/Biomes/Chrome/Spots/Spot_Chrome_Horror_Whoosh_Granualised_Cue
root=QualityLevel_0
nodes [
 {id=QualityLevel_0, children [Random_2,Random_2,Random_2,Random_2]}
 {id=Random_2, children [WavePlayer_39, WavePlayer_40, WavePlayer_41], properties {bShouldExcludeFromBranchCulling=true, weights [1,1,1]}}
 {id=WavePlayer_39, properties {soundWaveAssetPtr=&/Spot_Chrome_Horror_Whoosh_Granualised_04}}
 {id=WavePlayer_40, properties {soundWaveAssetPtr=&/Spot_Chrome_Horror_Whoosh_Granualised_03}}
 {id=WavePlayer_41, properties {soundWaveAssetPtr=&/Spot_Chrome_Horror_Whoosh_Granualised_02}}
]
properties {
 attenuationSettings=/Game/Sounds/Attenuation/Fort_TileAmbient_OneShot_Wind_Att
 bEnableBusSends=false
 maxDistance=8000
 soundClassObject=&Fort_Ambient_OneShots
 volumeMultiplier=0.401
 preEffectBusSends [{}]
 soundSubmixSends [{maxSendLevel=0.5, minSendLevel=0, sendLevel=0, soundSubmix=&Fortnite_Reverb_Distant}]
}
