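# Ambient spot sound cue for the Chrome biome, playing granular industrial drone vocal variations.
# An equally weighted Random node picks from fifteen waves, four of which pass through single-input Mixers that boost their level.
# All quality levels share that Random node, and a root Modulator varies pitch and volume.
# Output uses the outdoor ambient sound class and a one-shot tile-ambient wind attenuation.
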SoundCue /Game/Sounds/Artemis/Biomes/Chrome/Spots/Spot_Chrome_Industrial_Drone_Vocals_Granular_Cue
root=Modulator_0
nodes [
 {id=Mixer_0, children [WavePlayer_10], properties {inputVolume [2]}}
 {id=Mixer_1, children [WavePlayer_12], properties {inputVolume [2]}}
 {id=Mixer_2, children [WavePlayer_13], properties {inputVolume [2]}}
 {id=Mixer_3, children [WavePlayer_15], properties {inputVolume [2]}}
 {id=Modulator_0, children [QualityLevel_0], properties {pitchMin=0.8, volumeMax=1.1}}
 {id=QualityLevel_0, children [Random_1,Random_1,Random_1,Random_1]}
 {
  id=Random_1
  children [
   WavePlayer_11
   WavePlayer_14
   WavePlayer_17
   WavePlayer_18
   WavePlayer_19
   WavePlayer_20
   WavePlayer_21
   WavePlayer_22
   WavePlayer_23
   WavePlayer_24
   WavePlayer_25
   Mixer_0
   Mixer_1
   Mixer_2
   Mixer_3
  ]
  properties {weights [1,1,1,1,1,1,1,1,1,1,1,1,1,1,1]}
 }
 {id=WavePlayer_10, properties {soundWaveAssetPtr=&/Spot_Chrome_Industrial_Drone_Vocals_Granular_01}}
 {id=WavePlayer_11, properties {soundWaveAssetPtr=&/Spot_Chrome_Industrial_Drone_Vocals_Granular_02}}
 {id=WavePlayer_12, properties {soundWaveAssetPtr=&/Spot_Chrome_Industrial_Drone_Vocals_Granular_03}}
 {id=WavePlayer_13, properties {soundWaveAssetPtr=&/Spot_Chrome_Industrial_Drone_Vocals_Granular_04}}
 {id=WavePlayer_14, properties {soundWaveAssetPtr=&/Spot_Chrome_Industrial_Drone_Vocals_Granular_05}}
 {id=WavePlayer_15, properties {soundWaveAssetPtr=&/Spot_Chrome_Industrial_Drone_Vocals_Granular_06}}
 {id=WavePlayer_17, properties {soundWaveAssetPtr=&/Spot_Chrome_Industrial_Drone_Vocals_Granular_07}}
 {id=WavePlayer_18, properties {soundWaveAssetPtr=&/Spot_Chrome_Industrial_Drone_Vocals_Granular_08}}
 {id=WavePlayer_19, properties {soundWaveAssetPtr=&/Spot_Chrome_Industrial_Drone_Vocals_Granular_09}}
 {id=WavePlayer_20, properties {soundWaveAssetPtr=&/Spot_Chrome_Industrial_Drone_Vocals_Granular_10}}
 {id=WavePlayer_21, properties {soundWaveAssetPtr=&/Spot_Chrome_Industrial_Drone_Vocals_Granular_11}}
 {id=WavePlayer_22, properties {soundWaveAssetPtr=&/Spot_Chrome_Industrial_Drone_Vocals_Granular_12}}
 {id=WavePlayer_23, properties {soundWaveAssetPtr=&/Spot_Chrome_Industrial_Drone_Vocals_Granular_13}}
 {id=WavePlayer_24, properties {soundWaveAssetPtr=&/Spot_Chrome_Industrial_Drone_Vocals_Granular_14}}
 {id=WavePlayer_25, properties {soundWaveAssetPtr=&/Spot_Chrome_Industrial_Drone_Vocals_Granular_15}}
]
properties {
 attenuationSettings=/Game/Sounds/Attenuation/Fort_TileAmbient_OneShot_Wind_Att
 maxDistance=8000
 soundClassObject=&Fort_Ambient_Outdoors_Athena
 volumeMultiplier=0.501
}
