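# Sound cue for an industrial vocal-mangle drone spot ambience in the Chrome biome.
# A Random node picks one of seven equally weighted drone waves, and a root Modulator varies its pitch with a lowered minimum.
# The Quality Level node between them routes all four inputs to the same Random node.
# Output is turned down, assigned to the outdoor ambient sound class, and uses the shared tile-ambient one-shot wind attenuation.
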SoundCue /Game/Sounds/Artemis/Biomes/Chrome/Spots/Spot_Chrome_Industrial_Vocal_Mangle_Drone_Cue
root=Modulator_0
nodes [
 {id=Modulator_0, children [QualityLevel_0], properties {pitchMin=0.8}}
 {id=QualityLevel_0, children [Random_1,Random_1,Random_1,Random_1]}
 {
  id=Random_1
  children [WavePlayer_11, WavePlayer_12, WavePlayer_10, WavePlayer_15, WavePlayer_17, WavePlayer_18, WavePlayer_19]
  properties {weights [1,1,1,1,1,1,1]}
 }
 {id=WavePlayer_10, properties {soundWaveAssetPtr=&/Spot_Chrome_Industrial_Vocal_Mangle_Drone_01}}
 {id=WavePlayer_11, properties {soundWaveAssetPtr=&/Spot_Chrome_Industrial_Vocal_Mangle_Drone_02}}
 {id=WavePlayer_12, properties {soundWaveAssetPtr=&/Spot_Chrome_Industrial_Vocal_Mangle_Drone_03}}
 {id=WavePlayer_15, properties {soundWaveAssetPtr=&/Spot_Chrome_Industrial_Vocal_Mangle_Drone_06}}
 {id=WavePlayer_17, properties {soundWaveAssetPtr=&/Spot_Chrome_Industrial_Vocal_Mangle_Drone_07}}
 {id=WavePlayer_18, properties {soundWaveAssetPtr=&/Spot_Chrome_Industrial_Vocal_Mangle_Drone_08}}
 {id=WavePlayer_19, properties {soundWaveAssetPtr=&/Spot_Chrome_Industrial_Vocal_Mangle_Drone_09}}
]
properties {
 attenuationSettings=/Game/Sounds/Attenuation/Fort_TileAmbient_OneShot_Wind_Att
 maxDistance=8000
 soundClassObject=&Fort_Ambient_Outdoors_Athena
 volumeMultiplier=0.351
}
