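# Chrome biome ambient one-shot spot cue that plays sandy wind gusts, some layered with a mysterious shimmer, through wind attenuation.
# A random pick chooses one of four mixes: two wind-only, two pairing half-volume shimmer with heavier sandy wind.
# Shimmer waves pass through one of two fade-out envelopes, then a shared looping pitch-drift envelope.
# A modulator varies overall pitch and volume, feeding only the first two quality-level slots.
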SoundCue /Game/Sounds/Artemis/Biomes/Chrome/Spots/Spot_Chrome_Mysterious_Shimmer_Cue
root=QualityLevel_0
nodes [
 {
  id=Enveloper_0
  children [Random_0]
  properties {
   pitchCurve {editorCurveData {keys [{time=0}]}}
   volumeCurve {editorCurveData {keys [
      <at 0 1>
      {arriveTangent=-0.0086, interpMode=RCIM_Cubic, leaveTangent=-0.0086, tangentMode=RCTM_User, time=1, value=1}
      <at 6.5 0 RCIM_Cubic>
     ]}}
  }
 }
 {
  id=Enveloper_1
  children [Random_7]
  properties {
   pitchCurve {editorCurveData {keys [{time=0}]}}
   volumeCurve {editorCurveData {keys [<at 0 1>, {interpMode=RCIM_Cubic, leaveTangent=-0.2, time=2, value=1}, <at 7.5 0 RCIM_Cubic>]}}
  }
 }
 {
  id=Enveloper_2
  children [Random_1]
  properties {
   bLoop=true
   loopEnd=20
   pitchCurve {editorCurveData {keys [
      {interpMode=RCIM_Cubic, time=0}
      {arriveTangent=-0.0004, interpMode=RCIM_Cubic, leaveTangent=-0.0004, tangentMode=RCTM_User, time=5, value=1.05}
      {arriveTangent=-0.0033, interpMode=RCIM_Cubic, leaveTangent=-0.0033, time=10, value=0.98}
      {interpMode=RCIM_Cubic, time=20}
     ]}}
   volumeCurve {editorCurveData {keys [<at 0 1>]}}
  }
 }
 {id=Mixer_0, children [Random_4,null], properties {inputVolume [1,1]}}
 {id=Mixer_2, children [Enveloper_2,Random_6], properties {inputVolume [0.5,1]}}
 {id=Mixer_3, children [Enveloper_2,Random_5], properties {inputVolume [0.5,1]}}
 {id=Mixer_4, children [Random_13,null], properties {inputVolume [1,1]}}
 {id=Modulator_1, children [Random_2], properties {pitchMax=1.1, pitchMin=1, volumeMax=1.1, volumeMin=0.9}}
 {id=QualityLevel_0, children [Modulator_1,Modulator_1,null,null]}
 {id=Random_0, children [WavePlayer_13, WavePlayer_14, WavePlayer_15, WavePlayer_16], properties {weights [1,1,1,1]}}
 {id=Random_1, children [Enveloper_0,Enveloper_1], properties {bShouldExcludeFromBranchCulling=true, weights [1,1]}}
 {id=Random_13, children [WavePlayer_68, WavePlayer_69, WavePlayer_70, WavePlayer_71, WavePlayer_72], properties {weights [1,1,1,1,1]}}
 {id=Random_2, children [Mixer_4,Mixer_0,Mixer_2,Mixer_3], properties {weights [1,1,1,1]}}
 {id=Random_4, children [WavePlayer_8, WavePlayer_9, WavePlayer_10, WavePlayer_11, WavePlayer_12], properties {weights [1,1,1,1,1]}}
 {
  id=Random_5
  children [
   WavePlayer_22
   WavePlayer_23
   WavePlayer_24
   WavePlayer_25
   WavePlayer_26
   WavePlayer_27
   WavePlayer_28
   WavePlayer_29
   WavePlayer_30
   WavePlayer_20
  ]
  properties {weights [1,1,1,1,1,1,1,1,1,1]}
 }
 {
  id=Random_6
  children [WavePlayer_31, WavePlayer_32, WavePlayer_33, WavePlayer_34, WavePlayer_35, WavePlayer_36, WavePlayer_37, WavePlayer_38]
  properties {weights [1,1,1,1,1,1,1,1]}
 }
 {id=Random_7, children [WavePlayer_17, WavePlayer_18, WavePlayer_19, WavePlayer_21], properties {weights [1,1,1,1]}}
 {id=WavePlayer_10, properties {soundWaveAssetPtr=/Game/Sounds/Apollo/Biomes/Desert/AMB_SS_S15_Desert_WindGust_Med_Lt_Debris_02}}
 {id=WavePlayer_11, properties {soundWaveAssetPtr=/Game/Sounds/Apollo/Biomes/Desert/AMB_SS_S15_Desert_WindGust_Med_Lt_Debris_03}}
 {id=WavePlayer_12, properties {soundWaveAssetPtr=/Game/Sounds/Apollo/Biomes/Desert/AMB_SS_S15_Desert_WindGust_Med_Lt_Debris_04}}
 {id=WavePlayer_13, properties {soundWaveAssetPtr=&/Spot_Chrome_Mysterious_Shimmer_04}}
 {id=WavePlayer_14, properties {soundWaveAssetPtr=&/Spot_Chrome_Mysterious_Shimmer_03}}
 {id=WavePlayer_15, properties {soundWaveAssetPtr=&/Spot_Chrome_Mysterious_Shimmer_02}}
 {id=WavePlayer_16, properties {soundWaveAssetPtr=&/Spot_Chrome_Mysterious_Shimmer_01}}
 {id=WavePlayer_17, properties {soundWaveAssetPtr=&Spot_Chrome_Mysterious_Shimmer_04}}
 {id=WavePlayer_18, properties {soundWaveAssetPtr=&Spot_Chrome_Mysterious_Shimmer_03}}
 {id=WavePlayer_19, properties {soundWaveAssetPtr=&Spot_Chrome_Mysterious_Shimmer_02}}
 {id=WavePlayer_20, properties {soundWaveAssetPtr=/Game/Sounds/Apollo/Biomes/Desert/AMB_SS_S15_Desert_WindGust_Med_hv_Sandy_10}}
 {id=WavePlayer_21, properties {soundWaveAssetPtr=&Spot_Chrome_Mysterious_Shimmer_01}}
 {id=WavePlayer_22, properties {soundWaveAssetPtr=/Game/Sounds/Apollo/Biomes/Desert/AMB_SS_S15_Desert_WindGust_Med_hv_Sandy_06}}
 {id=WavePlayer_23, properties {soundWaveAssetPtr=/Game/Sounds/Apollo/Biomes/Desert/AMB_SS_S15_Desert_WindGust_Med_hv_Sandy_01}}
 {id=WavePlayer_24, properties {soundWaveAssetPtr=/Game/Sounds/Apollo/Biomes/Desert/AMB_SS_S15_Desert_WindGust_Med_hv_Sandy_02}}
 {id=WavePlayer_25, properties {soundWaveAssetPtr=/Game/Sounds/Apollo/Biomes/Desert/AMB_SS_S15_Desert_WindGust_Med_hv_Sandy_03}}
 {id=WavePlayer_26, properties {soundWaveAssetPtr=/Game/Sounds/Apollo/Biomes/Desert/AMB_SS_S15_Desert_WindGust_Med_hv_Sandy_04}}
 {id=WavePlayer_27, properties {soundWaveAssetPtr=/Game/Sounds/Apollo/Biomes/Desert/AMB_SS_S15_Desert_WindGust_Med_hv_Sandy_05}}
 {id=WavePlayer_28, properties {soundWaveAssetPtr=/Game/Sounds/Apollo/Biomes/Desert/AMB_SS_S15_Desert_WindGust_Med_hv_Sandy_07}}
 {id=WavePlayer_29, properties {soundWaveAssetPtr=/Game/Sounds/Apollo/Biomes/Desert/AMB_SS_S15_Desert_WindGust_Med_hv_Sandy_08}}
 {id=WavePlayer_30, properties {soundWaveAssetPtr=/Game/Sounds/Apollo/Biomes/Desert/AMB_SS_S15_Desert_WindGust_Med_hv_Sandy_09}}
 {id=WavePlayer_31, properties {soundWaveAssetPtr=/Game/Sounds/Apollo/Biomes/Tower_Aftermath/AMB_Tower_AM_Sandy_Wind_Oneshot_01}}
 {id=WavePlayer_32, properties {soundWaveAssetPtr=/Game/Sounds/Apollo/Biomes/Tower_Aftermath/AMB_Tower_AM_Sandy_Wind_Oneshot_12}}
 {id=WavePlayer_33, properties {soundWaveAssetPtr=/Game/Sounds/Apollo/Biomes/Tower_Aftermath/AMB_Tower_AM_Sandy_Wind_Oneshot_02}}
 {id=WavePlayer_34, properties {soundWaveAssetPtr=/Game/Sounds/Apollo/Biomes/Tower_Aftermath/AMB_Tower_AM_Sandy_Wind_Oneshot_03}}
 {id=WavePlayer_35, properties {soundWaveAssetPtr=/Game/Sounds/Apollo/Biomes/Tower_Aftermath/AMB_Tower_AM_Sandy_Wind_Oneshot_04}}
 {id=WavePlayer_36, properties {soundWaveAssetPtr=/Game/Sounds/Apollo/Biomes/Tower_Aftermath/AMB_Tower_AM_Sandy_Wind_Oneshot_05}}
 {id=WavePlayer_37, properties {soundWaveAssetPtr=/Game/Sounds/Apollo/Biomes/Tower_Aftermath/AMB_Tower_AM_Sandy_Wind_Oneshot_06}}
 {id=WavePlayer_38, properties {soundWaveAssetPtr=/Game/Sounds/Apollo/Biomes/Tower_Aftermath/AMB_Tower_AM_Sandy_Wind_Oneshot_07}}
 {id=WavePlayer_68, properties {soundWaveAssetPtr=/Game/Sounds/Artemis/Biomes/Desert/Desert_Spots/Spot_Desert_Wind_Sand_Gust_01}}
 {id=WavePlayer_69, properties {soundWaveAssetPtr=/Game/Sounds/Artemis/Biomes/Desert/Desert_Spots/Spot_Desert_Wind_Sand_Gust_02}}
 {id=WavePlayer_70, properties {soundWaveAssetPtr=/Game/Sounds/Artemis/Biomes/Desert/Desert_Spots/Spot_Desert_Wind_Sand_Gust_03}}
 {id=WavePlayer_71, properties {soundWaveAssetPtr=/Game/Sounds/Artemis/Biomes/Desert/Desert_Spots/Spot_Desert_Wind_Sand_Gust_04}}
 {id=WavePlayer_72, properties {soundWaveAssetPtr=/Game/Sounds/Artemis/Biomes/Desert/Desert_Spots/Spot_Desert_Wind_Sand_Gust_05}}
 {id=WavePlayer_8, properties {soundWaveAssetPtr=/Game/Sounds/Apollo/Biomes/Desert/AMB_SS_S15_Desert_WindGust_Med_Lt_Debris_05}}
 {id=WavePlayer_9, properties {soundWaveAssetPtr=/Game/Sounds/Apollo/Biomes/Desert/AMB_SS_S15_Desert_WindGust_Med_Lt_Debris_01}}
]
properties {
 attenuationSettings=/Game/Sounds/Attenuation/Fort_TileAmbient_OneShot_Wind_Att
 bEnableBusSends=false
 maxDistance=8000
 soundClassObject=&Fort_Ambient_OneShots
 volumeMultiplier=0.501
 preEffectBusSends [{}]
}
