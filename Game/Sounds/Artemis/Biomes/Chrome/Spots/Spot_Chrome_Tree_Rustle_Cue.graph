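# Sound cue for one-shot ambient tree rustle spot sounds in the Artemis Chrome biome, with wide tile-ambient attenuation.
# A quality-level root sends its first two slots to a modulator that varies pitch and volume over an equal-weight random pick of ten rustle waves.
# The last two quality slots point to a second modulator with no child, so those levels play nothing.
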SoundCue /Game/Sounds/Artemis/Biomes/Chrome/Spots/Spot_Chrome_Tree_Rustle_Cue
root=QualityLevel_0
nodes [
 {id=Modulator_1, children [Random_0], properties {pitchMax=1.1, pitchMin=1, volumeMax=1.1, volumeMin=0.5}}
 {id=Modulator_2, children [null], properties {pitchMax=1.1, volumeMax=1.1, volumeMin=0.1}}
 {id=QualityLevel_0, children [Modulator_1,Modulator_1,Modulator_2,Modulator_2]}
 {
  id=Random_0
  children [WavePlayer_0, WavePlayer_2, WavePlayer_3, WavePlayer_4, WavePlayer_5, WavePlayer_6, WavePlayer_7, WavePlayer_8, WavePlayer_9, WavePlayer_10]
  properties {weights [1,1,1,1,1,1,1,1,1,1]}
 }
 {id=WavePlayer_0, properties {soundWaveAssetPtr=&/Spot_Chrome_Trees_Rustle_01}}
 {id=WavePlayer_10, properties {soundWaveAssetPtr=&/Spot_Chrome_Trees_Rustle_11}}
 {id=WavePlayer_2, properties {soundWaveAssetPtr=&/Spot_Chrome_Trees_Rustle_03}}
 {id=WavePlayer_3, properties {soundWaveAssetPtr=&/Spot_Chrome_Trees_Rustle_04}}
 {id=WavePlayer_4, properties {soundWaveAssetPtr=&/Spot_Chrome_Trees_Rustle_05}}
 {id=WavePlayer_5, properties {soundWaveAssetPtr=&/Spot_Chrome_Trees_Rustle_06}}
 {id=WavePlayer_6, properties {soundWaveAssetPtr=&/Spot_Chrome_Trees_Rustle_07}}
 {id=WavePlayer_7, properties {soundWaveAssetPtr=&/Spot_Chrome_Trees_Rustle_08}}
 {id=WavePlayer_8, properties {soundWaveAssetPtr=&/Spot_Chrome_Trees_Rustle_09}}
 {id=WavePlayer_9, properties {soundWaveAssetPtr=&/Spot_Chrome_Trees_Rustle_10}}
]
properties {
 attenuationSettings=/Game/Sounds/Attenuation/Fort_TileAmbient_OneShot_Wider_Att
 bEnableBusSends=false
 maxDistance=8000
 soundClassObject=&Fort_Ambient_OneShots
 volumeMultiplier=0.501
 preEffectBusSends [{}]
}
