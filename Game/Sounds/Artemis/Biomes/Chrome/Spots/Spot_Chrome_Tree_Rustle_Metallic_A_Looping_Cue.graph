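# Looping ambient spot cue for metallic tree rustles in the Chrome biome, played as an attenuated one-shot ambient sound.
# A loop repeats a randomized delay before a quality switch, spacing out the rustles.
# The first two quality slots randomly pick one of ten equally weighted waves with slight pitch and volume modulation; the other two use an empty modulator, so are silent.
# A concurrency override limits playback to one instance.
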SoundCue /Game/Sounds/Artemis/Biomes/Chrome/Spots/Spot_Chrome_Tree_Rustle_Metallic_A_Looping_Cue
root=Looping_0
nodes [
 {id=Delay_0, children [QualityLevel_0], properties {delayMax=12, delayMin=6}}
 {id=Looping_0, children [Delay_0]}
 {id=Modulator_1, children [Random_0], properties {pitchMax=1.1, pitchMin=1, volumeMax=1, volumeMin=0.8}}
 {id=Modulator_2, children [null], properties {pitchMax=1.1, volumeMax=1.1, volumeMin=0.1}}
 {id=QualityLevel_0, children [Modulator_1,Modulator_1,Modulator_2,Modulator_2]}
 {
  id=Random_0
  children [WavePlayer_0,WavePlayer_1,WavePlayer_2,WavePlayer_3,WavePlayer_4,WavePlayer_5,WavePlayer_6,WavePlayer_7,WavePlayer_8,WavePlayer_9]
  properties {weights [1,1,1,1,1,1,1,1,1,1]}
 }
 {id=WavePlayer_0, properties {soundWaveAssetPtr=&/Spot_Chrome_Trees_Metallic_A_01}}
 {id=WavePlayer_1, properties {soundWaveAssetPtr=&/Spot_Chrome_Trees_Metallic_A_02}}
 {id=WavePlayer_2, properties {soundWaveAssetPtr=&/Spot_Chrome_Trees_Metallic_A_03}}
 {id=WavePlayer_3, properties {soundWaveAssetPtr=&/Spot_Chrome_Trees_Metallic_A_04}}
 {id=WavePlayer_4, properties {soundWaveAssetPtr=&/Spot_Chrome_Trees_Metallic_A_05}}
 {id=WavePlayer_5, properties {soundWaveAssetPtr=&/Spot_Chrome_Trees_Metallic_A_06}}
 {id=WavePlayer_6, properties {soundWaveAssetPtr=&/Spot_Chrome_Trees_Metallic_A_07}}
 {id=WavePlayer_7, properties {soundWaveAssetPtr=&/Spot_Chrome_Trees_Metallic_A_08}}
 {id=WavePlayer_8, properties {soundWaveAssetPtr=&/Spot_Chrome_Trees_Metallic_A_09}}
 {id=WavePlayer_9, properties {soundWaveAssetPtr=&/Spot_Chrome_Trees_Metallic_A_10}}
]
properties {
 attenuationSettings=/Game/Sounds/Attenuation/Fort_TileAmbient_OneShot_Tree_Rustle_Att
 bEnableBusSends=false
 bOverrideConcurrency=true
 maxDistance=6850
 soundClassObject=&Fort_Ambient_OneShots
 volumeMultiplier=0.601
 concurrencyOverrides {maxCount=1, voiceStealReleaseTime=2}
 preEffectBusSends [{}]
}
