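# Looping spot ambience cue for metallic tree rustles in the Chrome biome, using tree rustle one-shot attenuation.
# An endless loop waits a random 6 to 12 seconds, then a quality-level switch picks a branch.
# The first two quality slots play one of nine equally weighted waves with slight pitch and volume variation; the last two lead to an empty modulator, so stay silent.
# A concurrency override allows only one instance at once.
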SoundCue /Game/Sounds/Artemis/Biomes/Chrome/Spots/Spot_Chrome_Tree_Rustle_Metallic_B_Looping_Cue
root=Looping_0
nodes [
 {id=Delay_0, children [QualityLevel_0], properties {delayMax=12, delayMin=6}}
 {id=Looping_0, children [Delay_0]}
 {id=Modulator_1, children [Random_1], properties {pitchMax=1.1, pitchMin=1, volumeMax=1, volumeMin=0.8}}
 {id=Modulator_2, children [null], properties {pitchMax=1.1, volumeMax=1.1, volumeMin=0.1}}
 {id=QualityLevel_0, children [Modulator_1,Modulator_1,Modulator_2,Modulator_2]}
 {
  id=Random_1
  children [WavePlayer_10, WavePlayer_11, WavePlayer_12, WavePlayer_13, WavePlayer_14, WavePlayer_15, WavePlayer_16, WavePlayer_17, WavePlayer_18]
  properties {weights [1,1,1,1,1,1,1,1,1]}
 }
 {id=WavePlayer_10, properties {soundWaveAssetPtr=&/Spot_Chrome_Trees_Metallic_B_01}}
 {id=WavePlayer_11, properties {soundWaveAssetPtr=&/Spot_Chrome_Trees_Metallic_B_02}}
 {id=WavePlayer_12, properties {soundWaveAssetPtr=&/Spot_Chrome_Trees_Metallic_B_03}}
 {id=WavePlayer_13, properties {soundWaveAssetPtr=&/Spot_Chrome_Trees_Metallic_B_04}}
 {id=WavePlayer_14, properties {soundWaveAssetPtr=&/Spot_Chrome_Trees_Metallic_B_05}}
 {id=WavePlayer_15, properties {soundWaveAssetPtr=&/Spot_Chrome_Trees_Metallic_B_06}}
 {id=WavePlayer_16, properties {soundWaveAssetPtr=&/Spot_Chrome_Trees_Metallic_B_07}}
 {id=WavePlayer_17, properties {soundWaveAssetPtr=&/Spot_Chrome_Trees_Metallic_B_08}}
 {id=WavePlayer_18, properties {soundWaveAssetPtr=&/Spot_Chrome_Trees_Metallic_B_09}}
]
properties {
 attenuationSettings=/Game/Sounds/Attenuation/Fort_TileAmbient_OneShot_Tree_Rustle_Att
 bEnableBusSends=false
 bOverrideConcurrency=true
 maxDistance=6850
 soundClassObject=&Fort_Ambient_OneShots
 volumeMultiplier=0.601
 concurrencyOverrides {maxCount=1, voiceStealReleaseTime=3}
 preEffectBusSends [{}]
}
