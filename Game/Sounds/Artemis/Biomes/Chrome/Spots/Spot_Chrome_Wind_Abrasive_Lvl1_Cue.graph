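# Sound cue for abrasive level 1 wind spot ambience in the Chrome biome, classed as outdoor ambient.
# A random node picks among six equally weighted wave variations, and every quality level branch shares that same pool.
# A root modulator varies pitch with a lowered minimum, on top of reduced cue-wide pitch and volume multipliers.
# Spatial falloff uses the shared one-shot wind tile ambient attenuation asset.
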SoundCue /Game/Sounds/Artemis/Biomes/Chrome/Spots/Spot_Chrome_Wind_Abrasive_Lvl1_Cue
root=Modulator_0
nodes [
 {id=Modulator_0, children [QualityLevel_0], properties {pitchMin=0.8}}
 {id=QualityLevel_0, children [Random_2,Random_2,Random_2,Random_2]}
 {
  id=Random_2
  children [WavePlayer_10, WavePlayer_11, WavePlayer_12, WavePlayer_13, WavePlayer_14, WavePlayer_15]
  properties {weights [1,1,1,1,1,1]}
 }
 {id=WavePlayer_10, properties {soundWaveAssetPtr=&/Spot_Wind_Chrome_Abrasive_Lvl1_01_01}}
 {id=WavePlayer_11, properties {soundWaveAssetPtr=&/Spot_Wind_Chrome_Abrasive_Lvl1_01_02}}
 {id=WavePlayer_12, properties {soundWaveAssetPtr=&/Spot_Wind_Chrome_Abrasive_Lvl1_01_03}}
 {id=WavePlayer_13, properties {soundWaveAssetPtr=&/Spot_Wind_Chrome_Abrasive_Lvl1_02_01}}
 {id=WavePlayer_14, properties {soundWaveAssetPtr=&/Spot_Wind_Chrome_Abrasive_Lvl1_02_02}}
 {id=WavePlayer_15, properties {soundWaveAssetPtr=&/Spot_Wind_Chrome_Abrasive_Lvl1_02_03}}
]
properties {
 attenuationSettings=/Game/Sounds/Attenuation/Fort_TileAmbient_OneShot_Wind_Att
 maxDistance=8000
 pitchMultiplier=0.8
 soundClassObject=&Fort_Ambient_Outdoors_Athena
 volumeMultiplier=0.331
}
