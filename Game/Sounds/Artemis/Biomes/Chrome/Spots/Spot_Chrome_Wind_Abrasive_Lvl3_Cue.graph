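# Sound cue for an abrasive level 3 wind spot one-shot in the Chrome biome, classed as outdoor ambient audio.
# A Random node picks among six equally weighted wind waves.
# All four inputs of the Quality Level node share that Random node, so every tier plays the same set.
# A root Modulator varies playback with a lowered minimum pitch, and a tile-ambient wind one-shot attenuation asset handles distance falloff.
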SoundCue /Game/Sounds/Artemis/Biomes/Chrome/Spots/Spot_Chrome_Wind_Abrasive_Lvl3_Cue
root=Modulator_0
nodes [
 {id=Modulator_0, children [QualityLevel_0], properties {pitchMin=0.8}}
 {id=QualityLevel_0, children [Random_1,Random_1,Random_1,Random_1]}
 {
  id=Random_1
  children [WavePlayer_6, WavePlayer_7, WavePlayer_8, WavePlayer_9, WavePlayer_10, WavePlayer_11]
  properties {weights [1,1,1,1,1,1]}
 }
 {id=WavePlayer_10, properties {soundWaveAssetPtr=&/Spot_Wind_Chrome_Abrasive_Lvl3_02_02}}
 {id=WavePlayer_11, properties {soundWaveAssetPtr=&/Spot_Wind_Chrome_Abrasive_Lvl3_02_03}}
 {id=WavePlayer_6, properties {soundWaveAssetPtr=&/Spot_Wind_Chrome_Abrasive_Lvl3_01_01}}
 {id=WavePlayer_7, properties {soundWaveAssetPtr=&/Spot_Wind_Chrome_Abrasive_Lvl3_01_02}}
 {id=WavePlayer_8, properties {soundWaveAssetPtr=&/Spot_Wind_Chrome_Abrasive_Lvl3_01_03}}
 {id=WavePlayer_9, properties {soundWaveAssetPtr=&/Spot_Wind_Chrome_Abrasive_Lvl3_02_01}}
]
properties {
 attenuationSettings=/Game/Sounds/Attenuation/Fort_TileAmbient_OneShot_Wind_Att
 maxDistance=8000
 soundClassObject=&Fort_Ambient_Outdoors_Athena
 volumeMultiplier=0.251
}
