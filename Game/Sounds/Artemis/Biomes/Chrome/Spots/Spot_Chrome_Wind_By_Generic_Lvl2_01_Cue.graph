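# Sound cue for one-shot wind spot ambience in the Chrome biome, using outdoor ambient sound class and tile-ambient wind attenuation.
# A Random node picks among ten equally weighted wind waves, all Lvl1 recordings despite the cue's Lvl2 name.
# All four Quality Level inputs share that Random node, so quality settings do not change the content.
# A root Modulator varies playback with a lowered minimum pitch, and the cue volume is reduced.
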SoundCue /Game/Sounds/Artemis/Biomes/Chrome/Spots/Spot_Chrome_Wind_By_Generic_Lvl2_01_Cue
root=Modulator_0
nodes [
 {id=Modulator_0, children [QualityLevel_0], properties {pitchMin=0.8}}
 {id=QualityLevel_0, children [Random_0,Random_0,Random_0,Random_0]}
 {
  id=Random_0
  children [WavePlayer_0,WavePlayer_1,WavePlayer_2,WavePlayer_3,WavePlayer_4,WavePlayer_5,WavePlayer_6,WavePlayer_7,WavePlayer_8,WavePlayer_9]
  properties {weights [1,1,1,1,1,1,1,1,1,1]}
 }
 {id=WavePlayer_0, properties {soundWaveAssetPtr=&/Spot_Wind_Chrome_Generic_Lvl1_01}}
 {id=WavePlayer_1, properties {soundWaveAssetPtr=&/Spot_Wind_Chrome_Generic_Lvl1_02}}
 {id=WavePlayer_2, properties {soundWaveAssetPtr=&/Spot_Wind_Chrome_Generic_Lvl1_03}}
 {id=WavePlayer_3, properties {soundWaveAssetPtr=&/Spot_Wind_Chrome_Generic_Lvl1_04}}
 {id=WavePlayer_4, properties {soundWaveAssetPtr=&/Spot_Wind_Chrome_Generic_Lvl1_05}}
 {id=WavePlayer_5, properties {soundWaveAssetPtr=&/Spot_Wind_Chrome_Generic_Lvl1_06}}
 {id=WavePlayer_6, properties {soundWaveAssetPtr=&/Spot_Wind_Chrome_Generic_Lvl1_07}}
 {id=WavePlayer_7, properties {soundWaveAssetPtr=&/Spot_Wind_Chrome_Generic_Lvl1_08}}
 {id=WavePlayer_8, properties {soundWaveAssetPtr=&/Spot_Wind_Chrome_Generic_Lvl1_09}}
 {id=WavePlayer_9, properties {soundWaveAssetPtr=&/Spot_Wind_Chrome_Generic_Lvl1_10}}
]
properties {
 attenuationSettings=/Game/Sounds/Attenuation/Fort_TileAmbient_OneShot_Wind_Att
 maxDistance=8000
 soundClassObject=&Fort_Ambient_Outdoors_Athena
 volumeMultiplier=0.501
}
