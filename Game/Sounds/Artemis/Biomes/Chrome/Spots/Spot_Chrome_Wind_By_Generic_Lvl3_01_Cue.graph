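# Sound cue for level 3 generic wind-by spot one-shots in the Chrome biome's outdoor ambience.
# A Random node picks one of ten equally weighted wind wave variations, and a Modulator at the root adds slight pitch variation.
# All four Quality Level inputs share that same Random node, so every quality tier draws from the same pool.
# Playback uses a one-shot wind attenuation asset and the Athena outdoor ambient sound class at reduced volume.
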SoundCue /Game/Sounds/Artemis/Biomes/Chrome/Spots/Spot_Chrome_Wind_By_Generic_Lvl3_01_Cue
root=Modulator_0
nodes [
 {id=Modulator_0, children [QualityLevel_0], properties {pitchMax=1.1, pitchMin=0.9}}
 {id=QualityLevel_0, children [Random_1,Random_1,Random_1,Random_1]}
 {
  id=Random_1
  children [
   WavePlayer_10
   WavePlayer_11
   WavePlayer_12
   WavePlayer_13
   WavePlayer_14
   WavePlayer_15
   WavePlayer_16
   WavePlayer_17
   WavePlayer_18
   WavePlayer_19
  ]
  properties {weights [1,1,1,1,1,1,1,1,1,1]}
 }
 {id=WavePlayer_10, properties {soundWaveAssetPtr=&/Spot_Wind_Chrome_Generic_Lvl3_01}}
 {id=WavePlayer_11, properties {soundWaveAssetPtr=&/Spot_Wind_Chrome_Generic_Lvl3_02}}
 {id=WavePlayer_12, properties {soundWaveAssetPtr=&/Spot_Wind_Chrome_Generic_Lvl3_03}}
 {id=WavePlayer_13, properties {soundWaveAssetPtr=&/Spot_Wind_Chrome_Generic_Lvl3_04}}
 {id=WavePlayer_14, properties {soundWaveAssetPtr=&/Spot_Wind_Chrome_Generic_Lvl3_05}}
 {id=WavePlayer_15, properties {soundWaveAssetPtr=&/Spot_Wind_Chrome_Generic_Lvl3_06}}
 {id=WavePlayer_16, properties {soundWaveAssetPtr=&/Spot_Wind_Chrome_Generic_Lvl3_07}}
 {id=WavePlayer_17, properties {soundWaveAssetPtr=&/Spot_Wind_Chrome_Generic_Lvl3_08}}
 {id=WavePlayer_18, properties {soundWaveAssetPtr=&/Spot_Wind_Chrome_Generic_Lvl3_09}}
 {id=WavePlayer_19, properties {soundWaveAssetPtr=&/Spot_Wind_Chrome_Generic_Lvl3_10}}
]
properties {
 attenuationSettings=/Game/Sounds/Attenuation/Fort_TileAmbient_OneShot_Wind_Att
 maxDistance=8000
 soundClassObject=&Fort_Ambient_Outdoors_Athena
 volumeMultiplier=0.501
}
