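# Sound cue for a dark, sandy, level 1 wind spot ambience in the Chrome biome, classed as an ambient one-shot with wind attenuation.
# An equal-weight Random node picks one of five wind waves, and every quality level routes to that same node.
# A Modulator randomizes pitch, feeding an indefinitely looping Enveloper with a subtle cubic pitch curve.
# The cue plays at about half volume with bus sends disabled.
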SoundCue /Game/Sounds/Artemis/Biomes/Chrome/Spots/Spot_Chrome_Wind_Dark_Sandy_Lvl1_Cue
root=Enveloper_0
nodes [
 {
  id=Enveloper_0
  children [Modulator_0]
  properties {
   bLoop=true
   bLoopIndefinitely=true
   loopEnd=15
   pitchCurve {editorCurveData {keys [
      {interpMode=RCIM_Cubic, time=0}
      <at 15 1.02 RCIM_Cubic>
      {arriveTangent=0, interpMode=RCIM_Cubic, leaveTangent=0, tangentMode=RCTM_User, time=30}
      <at 45 1.05 RCIM_Cubic>
      {time=60}
     ]}}
   volumeCurve {editorCurveData {keys [<at 0 1>]}}
  }
 }
 {id=Modulator_0, children [QualityLevel_0], properties {pitchMin=0.8}}
 {id=QualityLevel_0, children [Random_1,Random_1,Random_1,Random_1]}
 {id=Random_1, children [WavePlayer_11, WavePlayer_12, WavePlayer_13, WavePlayer_14, WavePlayer_15], properties {weights [1,1,1,1,1]}}
 {id=WavePlayer_11, properties {soundWaveAssetPtr=&/Spot_Wind_Chrome_Dark_Sandy_Lvl1_01_02}}
 {id=WavePlayer_12, properties {soundWaveAssetPtr=&/Spot_Wind_Chrome_Dark_Sandy_Lvl1_01_03}}
 {id=WavePlayer_13, properties {soundWaveAssetPtr=&/Spot_Wind_Chrome_Dark_Sandy_Lvl1_02_01}}
 {id=WavePlayer_14, properties {soundWaveAssetPtr=&/Spot_Wind_Chrome_Dark_Sandy_Lvl1_02_02}}
 {id=WavePlayer_15, properties {soundWaveAssetPtr=&/Spot_Wind_Chrome_Dark_Sandy_Lvl1_02_03}}
]
properties {
 attenuationSettings=/Game/Sounds/Attenuation/Fort_TileAmbient_OneShot_Wind_Att
 bEnableBusSends=false
 maxDistance=8000
 soundClassObject=&Fort_Ambient_OneShots
 volumeMultiplier=0.501
 preEffectBusSends [{}]
}
