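# Sound cue for a Chrome biome ambient one-shot spot sound, a dark-twinkle wind variant at level 1.
# An equally weighted Random node picks one of ten wave variations.
# A quality-level node routes all four tiers to that same random pool, and a Modulator on top adds slight volume variation.
# It plays at reduced volume in the ambient one-shots class with a wind one-shot attenuation asset.
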SoundCue /Game/Sounds/Artemis/Biomes/Chrome/Spots/Spot_Wind_Chrome_DarkTwinkle_Lvl1_03_Cue
root=Modulator_0
nodes [
 {id=Modulator_0, children [QualityLevel_0], properties {volumeMax=1.1, volumeMin=1}}
 {id=QualityLevel_0, children [Random_1,Random_1,Random_1,Random_1]}
 {
  id=Random_1
  children [
   WavePlayer_10
   WavePlayer_11
   WavePlayer_12
   WavePlayer_13
   WavePlayer_14
   WavePlayer_15
   WavePlayer_16
   WavePlayer_17
   WavePlayer_18
   WavePlayer_19
  ]
  properties {weights [1,1,1,1,1,1,1,1,1,1]}
 }
 {id=WavePlayer_10, properties {soundWaveAssetPtr=&/Spot_Wind_Chrome_DarkTwinkle_Lvl1_03_01}}
 {id=WavePlayer_11, properties {soundWaveAssetPtr=&/Spot_Wind_Chrome_DarkTwinkle_Lvl1_03_02}}
 {id=WavePlayer_12, properties {soundWaveAssetPtr=&/Spot_Wind_Chrome_DarkTwinkle_Lvl1_03_03}}
 {id=WavePlayer_13, properties {soundWaveAssetPtr=&/Spot_Wind_Chrome_DarkTwinkle_Lvl1_03_04}}
 {id=WavePlayer_14, properties {soundWaveAssetPtr=&/Spot_Wind_Chrome_DarkTwinkle_Lvl1_03_05}}
 {id=WavePlayer_15, properties {soundWaveAssetPtr=&/Spot_Wind_Chrome_DarkTwinkle_Lvl1_03_06}}
 {id=WavePlayer_16, properties {soundWaveAssetPtr=&/Spot_Wind_Chrome_DarkTwinkle_Lvl1_03_07}}
 {id=WavePlayer_17, properties {soundWaveAssetPtr=&/Spot_Wind_Chrome_DarkTwinkle_Lvl1_03_08}}
 {id=WavePlayer_18, properties {soundWaveAssetPtr=&/Spot_Wind_Chrome_DarkTwinkle_Lvl1_03_09}}
 {id=WavePlayer_19, properties {soundWaveAssetPtr=&/Spot_Wind_Chrome_DarkTwinkle_Lvl1_03_10}}
]
properties {
 attenuationSettings=/Game/Sounds/Attenuation/Fort_TileAmbient_OneShot_Wind_Att
 bEnableBusSends=false
 maxDistance=8000
 soundClassObject=&Fort_Ambient_OneShots
 volumeMultiplier=0.401
 preEffectBusSends [{}]
}
